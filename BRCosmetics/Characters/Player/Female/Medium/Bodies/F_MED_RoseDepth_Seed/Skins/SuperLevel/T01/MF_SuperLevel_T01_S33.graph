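# Material function that applies the S33 SuperLevel T01 style to the RoseDepth Seed female character, overriding incoming base color, emissive, roughness, metallic and normal.
# Two mask textures define primary, secondary, outline, hair, metal and other areas, toggled by static switches.
# Effects include panning biplanar patterns from pre-skinned position, flowmap distortion, channel-offset caustics, Fresnel-indexed thin film, frost and outline glow.
# An IsFlyHair switch overrides only base color and emissive.
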MaterialFunction /BRCosmetics/Characters/Player/Female/Medium/Bodies/F_MED_RoseDepth_Seed/Skins/SuperLevel/T01/MF_SuperLevel_T01_S33
graph {nodes [
  n1 = VectorParameter() {group=SuperLevel, parameterName=TintColor_BaseColor, channelNames {a=Intensity, b="", g="", r=""}, defaultValue=<rgba 1 1 1 1>}
  n2 = Multiply(A=@n1, B=@n1)
  n3 = NamedRerouteUsage()
  n4 = Constant() {r=1}
  n5 = Saturate(None=Multiply(A=@n2, B=Desaturation(None=@n3, Fraction=@n4)))
  n6 = NamedRerouteUsage()
  n7 = LinearInterpolate(
   A=StaticSwitchParameter(True=LinearInterpolate(A=@n5, B=@n3, Alpha=@n6), False=@n3) {defaultValue=true, group=SuperLevel_Control, parameterName=TintBaseColor}
   Alpha=NamedRerouteUsage()
  ) {constB=0.02}
  n8 = VectorParameter() {
   group=SuperLevel_T01
   parameterName=OutlineColor_BaseColor_Dark
   channelNames {a=Intensity, b="", g="", r=""}
   defaultValue=<rgba 0 0.2346 0.2674 1>
  }
  n9 = VectorParameter() {
   group=SuperLevel_T01
   parameterName=OutlineColor_BaseColor
   channelNames {a=Intensity, b="", g="", r=""}
   defaultValue=<rgba 0 0.4083 0.4653 1>
  }
  n10 = LinearInterpolate(A=@n8, B=Multiply(A=@n9, B=@n9), Alpha=Fresnel() {baseReflectFraction=0, exponent=1})
  n11 = NamedRerouteUsage()
  n12 = LinearInterpolate(
   A=StaticSwitchParameter(True=LinearInterpolate(A=@n7, B=@n10, Alpha=@n11), False=@n7) {defaultValue=true, group=SuperLevel_Control, parameterName=UseOutlineArea}
   Alpha=NamedRerouteUsage()
  ) {constB=0}
  Add(A=NamedRerouteUsage(), B=@n12)
  AtmosphericLightVector()
  n13 = Multiply(A=TextureCoordinate(), B=ScalarParameter() {defaultValue=4, group=SuperLevel_T01, parameterName=SnowTilling}) {constB=5}
  n14 = TextureSampleParameter2D(UVs=Multiply(A=TextureCoordinate(), B=ScalarParameter() {defaultValue=3, group=SuperLevel_T01, parameterName=SnowTillingBump}) {constB=5}) {
   group=SuperLevel_T01
   parameterName=FrostNoise
   samplerType=SAMPLERTYPE_Masks
   texture=/BRCosmetics/Accessories/FORT_Backpacks/Backpack_M_MED_SatinCheddar/Materials/T_SatinCheddar_mask
  }
  BumpOffset(Coordinate=@n13, Height=Multiply(A=@n14.B, B=ScalarParameter() {group=SuperLevel_T01, parameterName=SnowBumpoffset}))
  Constant3Vector() {constant=<rgba 0 0 1 0>}
  n15 = TextureSampleParameter2D(UVs=@n13) {
   group=SuperLevel_T01
   parameterName=FrostNoise
   samplerType=SAMPLERTYPE_Masks
   texture=&T_SatinCheddar_mask
  }
  n16 = Power(Base=@n15.B) {constExponent=5}
  n17 = Multiply(A=Time(), B=ScalarParameter() {defaultValue=0.5, group=SuperLevel, parameterName=NoiseChangeSpeed})
  ConstantBiasScale(None=Sine(None=Add(A=@n16, B=@n17)))
  Custom() {code="return View.DirectionalLightDirection;", description=DirectionalLightVector, inputs []}
  n18 = FunctionInput() {inputName=Material, inputType=FunctionInput_MaterialAttributes}
  n19 = NamedRerouteUsage()
  n20 = NamedRerouteUsage()
  n21 = GetMaterialAttributes(None=@n18) {attributeGetTypes [
    #1
    #2
    #3
    #4
    #5
   ]}
  n22 = NamedRerouteUsage()
  n23 = NamedRerouteUsage()
  n24 = MaterialFunctionCall(Normal=@n23, Flatness=Constant() {r=0.5}) {materialFunction=&FlattenNormal}
  n25 = NamedRerouteUsage()
  n26 = SetMaterialAttributes!(
   MaterialAttributes=@n21.MaterialAttributes
   BaseColor=NamedRerouteUsage()
   EmissiveColor=NamedRerouteUsage()
   Roughness=NamedRerouteUsage()
   Metallic=StaticSwitchParameter(True=LinearInterpolate(A=@n21.Metallic, Alpha=@n22), False=@n21.Metallic) {group=SuperLevel_Control, parameterName=UseMetalArea}
   Normal=StaticSwitchParameter(True=LinearInterpolate(A=@n23, B=@n24.Result, Alpha=@n25), False=@n23) {defaultValue=true, group=SuperLevel_Control, parameterName=UseSecondaryArea}
  ) {attributeSetTypes [
    #1
    #2
    #3
    #5
    #4
   ]}
  FunctionOutput(None=StaticSwitchParameter(
    True=SetMaterialAttributes!(MaterialAttributes=@n18, BaseColor=@n19, EmissiveColor=@n20) {attributeSetTypes [#1, #2]}
    False=@n26
   ) {group=SuperLevel_Control, parameterName=IsFlyHair})
  MaterialFunctionCall!(ObjectPivotLocation, ObjectPositionOffset=Constant3Vector() {constant=<rgba 0 0 110 0>}) {materialFunction=/Game/Characters/Player/Common/MaterialFunctions/Superlevel/ImprovedStabilizedScreenSpaceUVs}
  n27 = ScalarParameter() {defaultValue=1, group=SuperLevel_T01, parameterName=SnowIntensity}
  n28 = Multiply(A=@n27) {constB=0.8}
  MaterialFunctionCall(Value_Morning=@n28, Value_Day=@n27, Value_Evening=@n28, Value_Night=Multiply(A=@n27) {constB=0.5}) {materialFunction=&Scalar_Time_of_Day}
  n29 = VectorParameter() {group=SuperLevel_Hair, parameterName=HairColor_Dark, defaultValue=<rgba 0 0.3908 0.4549 1>}
  n30 = VectorParameter() {group=SuperLevel_Hair, parameterName=HairColor, defaultValue=<rgba 1 1 1 1>}
  n31 = Desaturation(None=NamedRerouteUsage(), Fraction=Constant() {r=1})
  n32 = LinearInterpolate(A=Multiply(A=@n29.RGB, B=@n29.RGB), B=Multiply(A=@n30.RGB, B=@n30.RGB), Alpha=@n31)
  Multiply(A=@n32, B=@n31)
  n33 = ComponentMask(None=NamedRerouteUsage()) {b=true, g=true}
  n34 = ScalarParameter() {defaultValue=70, group=SuperLevel_T01, parameterName=TillingSize}
  n35 = NamedRerouteDeclaration(None=Multiply(A=Divide(A=@n33, B=@n34), B=ScalarParameter() {defaultValue=1.2, group=SuperLevel_T01, parameterName=UV_Tilling_Screen})) {name=UV_Screen_GB, nodeColor=<rgba 0.8984 1 0 1>}
  n36 = Add(A=@n35, B=NamedRerouteUsage())
  n37 = Constant()
  n38 = ScalarParameter() {defaultValue=-0.025, group=SuperLevel_T01, parameterName=SpeedY01}
  n39 = NamedRerouteUsage()
  n40 = TextureSample(UVs=Panner(Coordinate=@n36, Speed=AppendVector(A=@n37, B=@n38)), Tex=@n39) {
   samplerSource=SSM_Wrap_WorldGroupSettings
   samplerType=SAMPLERTYPE_Masks
   texture=/BRCosmetics/Characters/Player/Female/Medium/Bodies/F_MED_RoseDepth_Seed/Skins/SuperLevel/Textures/T_SuperLevel_Pattern_T01_S33
  }
  n41 = Power(Base=@n40.R)
  n42 = Panner(Coordinate=@n36, Speed=AppendVector(A=Constant(), B=ScalarParameter() {defaultValue=-0.02, group=SuperLevel_T01, parameterName=SpeedY02}))
  n43 = TextureSample(UVs=@n42, Tex=@n39) {
   samplerSource=SSM_Wrap_WorldGroupSettings
   samplerType=SAMPLERTYPE_Masks
   texture=&T_SuperLevel_Pattern_T01_S33
  }
  n44 = VectorParameter() {group=SuperLevel_T01, parameterName=Pattern_Bright, defaultValue=<rgba 0.0854 0.4167 0.3394 1>}
  n45 = Multiply(
   A=SmoothStep(Value=Add(A=@n41, B=@n43.B)) {constMax=0.8, constMin=0.18}
   B=LinearInterpolate(
    A=VectorParameter() {group=SuperLevel_T01, parameterName=Pattern_Dark, defaultValue=<rgba 0.1087 0.2919 0.5694 1>}.RGB
    B=NamedRerouteDeclaration(None=@n44.RGB) {name=Pattern_BrightValue_GB, nodeColor=<rgba 1 0 0.0234 1>}
    Alpha=Fresnel() {baseReflectFraction=0, exponent=1}
   )
  )
  Multiply(A=ScalarParameter() {defaultValue=3, group=SuperLevel_T01, parameterName=PrimaryEM_Intensity}, B=@n45)
  n46 = TextureSampleParameter2D() {
   group=SuperLevel_T01
   parameterName=SuperLevel_Mask02
   samplerType=SAMPLERTYPE_Masks
   texture=/BRCosmetics/Characters/Player/Female/Medium/Bodies/F_MED_RoseDepth_Seed/Textures/T_RoseDepth_Seed_Body_SL_FX2
  }
  n47 = TextureSampleParameter2D() {
   group=SuperLevel_T01
   parameterName=SuperLevel_Mask01
   samplerType=SAMPLERTYPE_Masks
   texture=/BRCosmetics/Characters/Player/Female/Medium/Bodies/F_MED_RoseDepth_Seed/Textures/T_RoseDepth_Seed_Body_SL_FX
  }
  n48 = NamedRerouteUsage()
  NamedRerouteDeclaration(None=StaticSwitchParameter(True=@n46.RGB, False=StaticSwitchParameter(True=@n47.RGB, False=@n48) {group=SuperLevel_T01, parameterName=DebugSuperLevelMask01}) {group=SuperLevel_T01, parameterName=DebugSuperLevelMask02}) {name=BaseColorOUT, nodeColor=<rgba 0.7812 1 0 1>}
  NamedRerouteDeclaration(None=@n21.BaseColor) {name=BaseColorIN, nodeColor=<rgba 0 1 0.5781 1>}
  n49 = ComponentMask(None=Multiply(A=Subtract(A=@n15.RGB) {constB=0.5}) {constB=2}) {g=true, r=true}
  NamedRerouteDeclaration(None=Transform(None=DeriveNormalZ(InXY=@n49))) {name=FrostNormal, nodeColor=<rgba 0 0.7656 1 1>}
  NamedRerouteDeclaration(None=@n21.Normal) {name=Normal_IN, nodeColor=<rgba 1 0 0.7734 1>}
  n50 = NamedRerouteUsage()
  n51 = ScalarParameter() {
   desc="LockedZ only for the specular highlight.\r\nfor a streched highlight in on the Z axis use the Blinn highlight"
   group=02_Anime Parameters
   parameterName=SpecularHighlight_LockedZVector
   sliderMax=1
  }
  n52 = AppendVector(A=Constant2Vector() {g=1, r=1}, B=Saturate(None=OneMinus(None=@n51)))
  n53 = Multiply(A=Add(A=CameraVectorWS(), B=@n50), B=@n52)
  n54 = Transform(None=NamedRerouteUsage())
  n55 = Multiply(A=@n50, B=@n52)
  n56 = StaticSwitchParameter(
   True=DotProduct(A=Normalize(VectorInput=@n53), B=@n54) {desc=Blinn}
   False=DotProduct(A=Normalize(VectorInput=@n55), B=ReflectionVectorWS(CustomWorldNormal=@n54)) {desc=Phong}
  ) {
   desc="Default is the Phong highlight.\r\nUse Blinn for a vertical streched highlight"
   group=02_Anime Parameters
   parameterName=UseBlinnHighlight
   sortPriority=0
  }
  n57 = DotProduct(A=@n54, B=CameraVectorWS())
  n58 = Saturate(None=LinearInterpolate(
    A=@n56
    B=NamedRerouteDeclaration(None=@n57) {name=NdotV, nodeColor=<rgba 1 0.6242 0 1>}
    Alpha=ScalarParameter() {group=02_Anime Parameters, parameterName=SpecularHighlight_CameraFacing_Blend, sliderMax=1}
   ))
  n59 = Constant() {r=1}
  n60 = ScalarParameter() {defaultValue=0.3, group=02_Anime Parameters, parameterName=SpecularHighlight_Size, sliderMax=1, sliderMin=0.0001}
  n61 = ScalarParameter() {group=02_Anime Parameters, parameterName=SpecularHighlight_Hardness, sliderMax=1}
  NamedRerouteDeclaration(None=Multiply(
    A=SphereMask(A=@n58, B=@n59, Radius=@n60, Hardness=@n61) {hardnessPercent=78}
    B=ScalarParameter() {defaultValue=1, group=02_Anime Parameters, parameterName=SpecularHighlight_Intensity, sliderMax=8}
   )) {name=Specular_Curve, nodeColor=<rgba 1 0.4453 0 1>}
  NamedRerouteDeclaration(None=@n16) {name=BaseColorNoise, nodeColor=<rgba 1 0 0.9844 1>}
  n62 = Fresnel() {baseReflectFraction=0, exponent=1}
  n63 = Sine(None=Multiply(A=@n62) {constB=3})
  n64 = Saturate(None=Multiply(A=ConstantBiasScale(None=@n63), B=Power(Base=@n62) {constExponent=3}))
  NamedRerouteDeclaration(None=@n64) {name=Normal_LUT, nodeColor=<rgba 1 0 0.3047 1>}
  n65 = Distance(A=Abs(None=NamedRerouteUsage()), B=Constant3Vector() {constant=<rgba 0 0 1 0>})
  n66 = Multiply(
   A=SmoothStep(
    Min=ScalarParameter() {defaultValue=0.2, group=SuperLevel_T01, parameterName=SeamMin}
    Max=ScalarParameter() {defaultValue=1, group=SuperLevel_T01, parameterName=SeamMax}
    Value=@n65
   )
   B=ScalarParameter() {defaultValue=0.2, group=SuperLevel_T01, parameterName=distanceMulti}
  )
  n67 = VectorParameter() {
   group=SuperLevel_T01
   parameterName=SecondaryColor_Seam
   channelNames {a=Intensity, b="", g="", r=""}
   defaultValue=<rgba 0.2493 0.7466 0.816 1>
  }
  n68 = Add(
   A=Multiply(A=@n27, B=NamedRerouteUsage())
   B=Multiply(A=LinearInterpolate(Alpha=@n64) {constA=0.1}, B=ScalarParameter() {defaultValue=0.8, group=SuperLevel_T01, parameterName=PlasticIntensity})
  )
  n69 = VectorParameter() {
   group=SuperLevel_T01
   parameterName=SecondaryColor_EM
   channelNames {a=Intensity, b="", g="", r=""}
   defaultValue=<rgba 0.629 0.7931 0.816 1>
  }
  n70 = Multiply(A=NamedRerouteUsage(), B=Add(A=Multiply(A=@n66, B=@n67), B=Multiply(A=@n68, B=@n69)))
  n71 = ScalarParameter() {defaultValue=1, group=SuperLevel_T01, parameterName=SecondaryColorControl}
  n72 = Multiply(A=@n71) {constB=0.8}
  n73 = MaterialFunctionCall(Value_Morning=@n72, Value_Day=@n71, Value_Evening=@n72, Value_Night=Multiply(A=@n71) {constB=0.5}) {materialFunction=&Scalar_Time_of_Day}
  NamedRerouteDeclaration(None=StaticSwitchParameter(True=Multiply(A=@n70, B=@n73.Result), False=Constant()) {defaultValue=true, group=SuperLevel_Control, parameterName=UseSecondaryArea}) {name=ShadingEM, nodeColor=<rgba 0 1 0.9766 1>}
  n74 = LinearInterpolate(Alpha=NamedRerouteUsage()) {constA=1, constB=5}
  n75 = ScalarParameter() {defaultValue=3, group=SuperLevel_T01, parameterName=PrimaryEM_Intensity}
  n76 = Divide(
   A=ComponentMask(None=NamedRerouteUsage()) {b=true, r=true}
   B=ScalarParameter() {defaultValue=70, group=SuperLevel_T01, parameterName=TillingSize}
  )
  n77 = ScalarParameter() {defaultValue=1.2, group=SuperLevel_T01, parameterName=UV_Tilling_Screen}
  n78 = Add(A=NamedRerouteDeclaration(None=Multiply(A=@n76, B=@n77)) {name=UV_Screen, nodeColor=<rgba 0.8984 1 0 1>}, B=NamedRerouteUsage())
  n79 = Constant()
  n80 = ScalarParameter() {defaultValue=-0.025, group=SuperLevel_T01, parameterName=SpeedY01}
  n81 = TextureObjectParameter() {
   group=SuperLevel_T01
   parameterName=SuperLevelPattern
   samplerType=SAMPLERTYPE_Masks
   texture=&T_SuperLevel_Pattern_T01_S33
  }
  n82 = TextureSample(UVs=Panner(Coordinate=@n78, Speed=AppendVector(A=@n79, B=@n80)), Tex=@n81) {
   samplerSource=SSM_Wrap_WorldGroupSettings
   samplerType=SAMPLERTYPE_Masks
   texture=&T_SuperLevel_Pattern_T01_S33
  }
  n83 = Power(Base=@n82.R)
  n84 = Panner(Coordinate=@n78, Speed=AppendVector(A=Constant(), B=ScalarParameter() {defaultValue=-0.02, group=SuperLevel_T01, parameterName=SpeedY02}))
  n85 = TextureSample(UVs=@n84, Tex=@n81) {
   samplerSource=SSM_Wrap_WorldGroupSettings
   samplerType=SAMPLERTYPE_Masks
   texture=&T_SuperLevel_Pattern_T01_S33
  }
  n86 = VectorParameter() {group=SuperLevel_T01, parameterName=Pattern_Bright, defaultValue=<rgba 0.0854 0.4167 0.3394 1>}
  n87 = Multiply(
   A=SmoothStep(Value=Add(A=@n83, B=@n85.B)) {constMax=0.8, constMin=0.18}
   B=LinearInterpolate(
    A=VectorParameter() {group=SuperLevel_T01, parameterName=Pattern_Dark, defaultValue=<rgba 0.1087 0.2919 0.5694 1>}.RGB
    B=NamedRerouteDeclaration(None=@n86.RGB) {name=Pattern_BrightValue, nodeColor=<rgba 1 0 0.0234 1>}
    Alpha=Fresnel() {baseReflectFraction=0, exponent=1}
   )
  )
  n88 = StaticSwitchParameter(True=LinearInterpolate(A=NamedRerouteUsage(), B=@n87, Alpha=NamedRerouteUsage()), False=@n87) {defaultValue=true, group=SuperLevel_Control, parameterName=UseBiplanarProjection}
  n89 = Multiply(A=LinearInterpolate(A=Constant(), B=Multiply(A=@n75, B=@n88), Alpha=NamedRerouteUsage()), B=NamedRerouteUsage())
  NamedRerouteDeclaration(None=Multiply(A=Multiply(A=@n74, B=@n89), B=ScalarParameter() {defaultValue=0.55, group=SuperLevel_Control, parameterName=PrimaryPatternIntensity})) {name=Pattern_Animated, nodeColor=<rgba 0.4531 1 0 1>}
  n90 = Custom() {code="return View.DirectionalLightDirection;", description=DirectionalLightVector, inputs []}
  n91 = DotProduct(A=@n90, B=@n90)
  n92 = CollectionParameter() {
   collection=&FortniteMaterialParameters
   parameterName=SunlightDirectionalVector
  }
  n93 = ComponentMask(None=@n92) {b=true, g=true, r=true}
  n94 = MaterialFunctionCall!(In=If(
    A=Custom() {code="return View.DirectionalLightColor;", description=DirectionalLightColor, inputs []}
    "A > B"=If(A=@n91, "A > B"=@n90, "A == B"=@n93, "A < B"=@n93) {desc=check if DirectionalLight exists}
    "A == B"=@n93
    "A < B"=@n93
   ) {desc=check if DirectionalLight color is Black}) {materialFunction=/Game/Packages/Fortress_SharedMaterials/Cinematics/Anime/MF_CINE_AnimeLightDirOverride}
  n95 = MaterialFunctionCall(Float3=@n94.Result) {materialFunction=&BreakOutFloat3Components}
  n96 = ScalarParameter() {
   desc="0.0 = the light vector is locked on the XY plane.\r\n1.0 = the actual light vector is used  \r\n(for Faces use a blend value of 0,0)"
   group=02_Anime Parameters
   parameterName=LockedZLightVector_Blend
   sliderMax=1
  }
  n97 = MaterialFunctionCall(X=@n95.R, Y=@n95.R, Z=Multiply(A=@n95.B, B=@n96)) {materialFunction=&MakeFloat3}
  n98 = StaticSwitchParameter(True=LinearInterpolate(A=@n94.Result, B=Normalize(VectorInput=@n97.Result), Alpha=VertexColor().A), False=@n94.Result) {
   desc="Lock the Z axis of the Light Vector\r\nis driven by the VertexColor A channel (0 = Regular LightVector, 1= Locked LIghtVector)"
   group=02_Anime Parameters
   parameterName=UseLockedZLightVector
   sortPriority=0
  }
  NamedRerouteDeclaration(None=@n98) {name=LightVector, nodeColor=<rgba 1 0.934 0.5103 1>}
  NamedRerouteDeclaration(None=SmoothStep(Value=@n82.G) {constMax=0.5}) {name=PanningG, nodeColor=<rgba 0.8047 1 0 1>}
  NamedRerouteDeclaration(None=@n47.G) {name=OutLineArea, nodeColor=<rgba 0 1 0.2623 1>}
  n99 = NamedRerouteUsage()
  n100 = ScalarParameter() {defaultValue=0.5, group=SuperLevel_T01, parameterName=DistortTilling}
  n101 = ScalarParameter() {defaultValue=-0.02, group=SuperLevel_T01, parameterName=SpeedY02}
  n102 = Constant()
  n103 = TextureSampleParameter2D(UVs=Panner(Coordinate=Multiply(A=@n99, B=@n100), Speed=AppendVector(A=@n101, B=@n102))) {
   group=SuperLevel_T01
   parameterName=Flowmap
   samplerType=SAMPLERTYPE_LinearColor
   texture=/Game/Global/Textures/Default/Flowmaps/T_SwirlFlow_01
  }
  NamedRerouteDeclaration(None=Multiply(
    A=ComponentMask(None=@n103.RGB) {g=true, r=true}
    B=ScalarParameter() {defaultValue=0.12, group=SuperLevel_T01, parameterName=DistortIntensity}
   )) {name=NormalDistort, nodeColor=<rgba 0 1 0.6953 1>}
  n104 = NamedRerouteDeclaration(None=@n46.G) {name=G_Channel, nodeColor=<rgba 0 0.1562 1 1>}
  NamedRerouteDeclaration(None=Saturate(None=Subtract(A=@n46.R, B=@n104))) {name=TertiaryArea, nodeColor=<rgba 1 0 0.75 1>}
  NamedRerouteDeclaration(None=SmoothStep(
    Min=ScalarParameter() {group=SuperLevel_T01, parameterName=SeamMin02}
    Max=ScalarParameter() {defaultValue=1, group=SuperLevel_T01, parameterName=SeamMax02}
    Value=@n65
   )) {name=DistanceShading, nodeColor=<rgba 0.1953 0 1 1>}
  n105 = NamedRerouteUsage()
  n106 = NamedRerouteUsage()
  n107 = NamedRerouteUsage()
  n108 = Add(A=@n106, B=Multiply(A=Power(Base=@n107)) {constB=0.5})
  n109 = Add(
   A=Multiply(A=Power(Base=@n105), B=ScalarParameter() {defaultValue=3, group=SuperLevel_T01, parameterName=DistanceColor_Intensity})
   B=Multiply(
    A=ScalarParameter() {defaultValue=0.5, group=SuperLevel_T01, parameterName=DistanceColor_Intensity_Fill}
    B=StaticSwitchParameter(True=@n108, False=@n106) {group=SuperLevel_Control, parameterName=UseAdditonal_Ter}
   ) {constA=0.2}
  ) {constB=0.1}
  n110 = VectorParameter() {group=SuperLevel_T01, parameterName=DistanceEM, defaultValue=<rgba 0 0.6473 1 1>}
  NamedRerouteDeclaration(None=Multiply(A=Multiply(A=@n109, B=@n110.RGB), B=NamedRerouteUsage())) {name=TertiaryEM, nodeColor=<rgba 0.125 1 0 1>}
  NamedRerouteDeclaration(None=@n21.Roughness) {name=RoughnessIn, nodeColor=<rgba 0 0.4844 1 1>}
  NamedRerouteDeclaration(None=@n84) {name=MainUV, nodeColor=<rgba 1 0 0.4219 1>}
  NamedRerouteDeclaration(None=@n81) {name=MainPattern, nodeColor=<rgba 0 0.7422 1 1>}
  NamedRerouteDeclaration(None=FunctionInput() {inputName=PreskinnedPos, sortPriority=1}) {name=PreskinnedPOS, nodeColor=<rgba 0.8516 1 0 1>}
  NamedRerouteDeclaration(None=@n21.EmissiveColor) {name=EM_IN, nodeColor=<rgba 0.5469 1 0 1>}
  n111 = ScalarParameter() {defaultValue=0.015, group=SuperLevel_T01, parameterName=CausticsIntensity}
  n112 = Constant() {r=1}
  n113 = Constant() {r=-1.732}
  n114 = Multiply(A=@n111, B=Normalize(VectorInput=AppendVector(A=@n112, B=@n113)))
  n115 = NamedRerouteUsage()
  n116 = NamedRerouteUsage()
  n117 = TextureSample(UVs=Add(A=@n114, B=@n115), Tex=@n116) {
   samplerSource=SSM_Wrap_WorldGroupSettings
   samplerType=SAMPLERTYPE_Masks
   texture=&T_SuperLevel_Pattern_T01_S33
  }
  n118 = Multiply(A=@n117.B, B=Constant3Vector() {constant=<rgba 1 0 0 0>})
  n119 = Multiply(A=@n111, B=Constant2Vector() {r=1})
  n120 = TextureSample(UVs=Add(A=@n119, B=@n115), Tex=@n116) {
   samplerSource=SSM_Wrap_WorldGroupSettings
   samplerType=SAMPLERTYPE_Masks
   texture=&T_SuperLevel_Pattern_T01_S33
  }
  n121 = Multiply(A=@n120.B, B=Constant3Vector() {constant=<rgba 0 1 0.2 1>})
  n122 = TextureSample(UVs=Add(A=@n119, B=@n115), Tex=@n116) {
   samplerSource=SSM_Wrap_WorldGroupSettings
   samplerType=SAMPLERTYPE_Masks
   texture=&T_SuperLevel_Pattern_T01_S33
  }
  n123 = Constant3Vector() {constant=<rgba 0 0.1 1 1>}
  n124 = Saturate(None=Add(A=Add(A=@n118, B=@n121), B=Multiply(A=@n122.B, B=@n123)))
  n125 = NamedRerouteUsage()
  n126 = NamedRerouteUsage()
  n127 = TextureSample(UVs=Add(A=@n114, B=@n125), Tex=@n126) {
   samplerSource=SSM_Wrap_WorldGroupSettings
   samplerType=SAMPLERTYPE_Masks
   texture=&T_SuperLevel_Pattern_T01_S33
  }
  n128 = Multiply(A=@n127.B, B=Constant3Vector() {constant=<rgba 1 0 0 0>})
  n129 = TextureSample(UVs=Add(A=@n119, B=@n125), Tex=@n126) {
   samplerSource=SSM_Wrap_WorldGroupSettings
   samplerType=SAMPLERTYPE_Masks
   texture=&T_SuperLevel_Pattern_T01_S33
  }
  n130 = Multiply(A=@n129.B, B=Constant3Vector() {constant=<rgba 0 1 0.2 1>})
  n131 = Normalize(VectorInput=AppendVector(A=Constant() {r=-1}, B=Constant() {r=-1.732}))
  n132 = TextureSample(UVs=Add(A=Multiply(A=@n111, B=@n131), B=@n125), Tex=@n126) {
   samplerSource=SSM_Wrap_WorldGroupSettings
   samplerType=SAMPLERTYPE_Masks
   texture=&T_SuperLevel_Pattern_T01_S33
  }
  n133 = Constant3Vector() {constant=<rgba 0 0.1 1 1>}
  n134 = Saturate(None=Add(A=Add(A=@n128, B=@n130), B=Multiply(A=@n132.B, B=@n133)))
  n135 = NamedRerouteUsage()
  NamedRerouteDeclaration(None=StaticSwitchParameter(True=LinearInterpolate(A=@n124, B=@n134, Alpha=@n135), False=@n134) {defaultValue=true, group=SuperLevel_Control, parameterName=UseBiplanarProjection}) {name=CausticsAlpha, nodeColor=<rgba 1 0.5625 0 1>}
  NamedRerouteDeclaration(None=@n46.B) {name=HairArea, nodeColor=<rgba 0.7344 0 1 1>}
  n136 = Multiply(A=@n46.R, B=NamedRerouteUsage())
  n137 = Saturate(None=Subtract(A=@n104, B=@n46.R))
  n138 = NamedRerouteUsage()
  NamedRerouteDeclaration(None=StaticSwitchParameter(True=@n136, False=StaticSwitchParameter(True=@n137, False=@n138) {group=SuperLevel_Control, parameterName=UseYellowAsMetal}) {group=SuperLevel_Control, parameterName=UseYellowAsShirt}) {name=DefaultEMArea, nodeColor=<rgba 0 1 0.6484 1>}
  n139 = Fresnel() {baseReflectFraction=0, exponent=1}
  n140 = Multiply(A=@n139, B=ScalarParameter() {defaultValue=0.5, group=SuperLevel_T01, parameterName=OutlineSine})
  n141 = Multiply(A=@n139, B=ConstantBiasScale(None=Sine(None=@n140)))
  n142 = VectorParameter() {group=SuperLevel_T01, parameterName=OutlineColor_EM, channelNames {a=Intensity, b="", g="", r=""}, defaultValue=<rgba 0 0.856 1 10>}
  n143 = Multiply(A=@n142, B=@n142)
  NamedRerouteDeclaration(None=Multiply(A=@n141, B=@n143)) {name=OutlineColor, nodeColor=<rgba 0.1719 1 0 1>}
  n144 = NamedRerouteUsage()
  NamedRerouteDeclaration(None=StaticSwitchParameter(True=@n137, False=StaticSwitchParameter(True=@n136, False=@n144) {group=SuperLevel_Control, parameterName=UseYellowAsMetal}) {group=SuperLevel_Control, parameterName=UseYellowAsShirt}) {name=MetalArea, nodeColor=<rgba 0 1 0.6484 1>}
  n145 = Saturate(None=@n32)
  NamedRerouteDeclaration(None=@n145) {name=HairColor_Adj, nodeColor=<rgba 0 1 0.4609 1>}
  n146 = VectorParameter() {group=SuperLevel_Hair, parameterName=HairColor_EM, defaultValue=<rgba 1 1 1 1>}
  n147 = Multiply(A=ScalarParameter() {defaultValue=1, group=SuperLevel_Control, parameterName=HairColor_EM_Override}, B=Multiply(A=@n146.RGB, B=@n146.RGB))
  NamedRerouteDeclaration(None=@n147) {name=HairEM_Adj, nodeColor=<rgba 0.0312 1 0 1>}
  NamedRerouteDeclaration(None=@n2) {name=TintColor_BaseColor, nodeColor=<rgba 1 0 0.4687 1>}
  NamedRerouteDeclaration(None=@n10) {name=OutlineColor_RE, nodeColor=<rgba 0.4297 1 0 1>}
  n148 = StaticSwitchParameter(True=LinearInterpolate(A=@n12, B=NamedRerouteUsage(), Alpha=NamedRerouteUsage()), False=@n12) {defaultValue=true, group=SuperLevel_Control, parameterName=UseSecondaryArea}
  n149 = StaticSwitchParameter(True=LinearInterpolate(A=@n148, B=@n145, Alpha=NamedRerouteUsage()), False=@n148) {group=SuperLevel_Control, parameterName=UseHairArea}
  n150 = VectorParameter() {group=SuperLevel_Metal, parameterName=MetalColor_Light, defaultValue=<rgba 0.8681 0.4813 0.1145 1>}
  n151 = VectorParameter() {group=SuperLevel_Metal, parameterName=MetalColor_Dark, defaultValue=<rgba 0.1597 0.0644 0.0194 1>}
  n152 = Power(Base=Fresnel() {baseReflectFraction=0, exponent=1}) {constExponent=0.5}
  n153 = ConstantBiasScale(None=Sine(None=Multiply(A=@n152) {constB=2}))
  n154 = StaticSwitchParameter(True=LinearInterpolate(A=@n149, B=LinearInterpolate(A=@n150.RGB, B=@n151.RGB, Alpha=@n153), Alpha=NamedRerouteUsage()), False=@n149) {group=SuperLevel_Control, parameterName=UseMetalArea}
  n155 = StaticSwitchParameter(True=LinearInterpolate(A=@n154, B=NamedRerouteUsage(), Alpha=NamedRerouteUsage()), False=@n154) {group=SuperLevel_Control, parameterName=UseYellowAsShirt}
  NamedRerouteDeclaration(None=@n155) {name=PostSuperLevel_BaseColor, nodeColor=<rgba 1 0 0.8672 1>}
  n156 = NamedRerouteUsage()
  n157 = Add(A=NamedRerouteUsage(), B=Multiply(A=Multiply(A=@n156) {constB=0.3}, B=NamedRerouteUsage()))
  n158 = Power(Base=NamedRerouteUsage()) {constExponent=4}
  n159 = NamedRerouteUsage()
  NamedRerouteDeclaration(None=Saturate(None=LinearInterpolate(A=@n157, B=@n158, Alpha=@n159))) {name=RoughnessOUT, nodeColor=<rgba 0 1 0.2031 1>}
  NamedRerouteDeclaration(None=Power(Base=@n14.B) {constExponent=20}) {name=SnowCover, nodeColor=<rgba 1 0 0.6328 1>}
  n160 = ScalarParameter() {defaultValue=0.5, group=SuperLevel_T01, parameterName=DistortTilling}
  n161 = ScalarParameter() {defaultValue=-0.02, group=SuperLevel_T01, parameterName=SpeedY02}
  n162 = Constant()
  n163 = TextureSampleParameter2D(UVs=Panner(Coordinate=Multiply(A=@n35, B=@n160), Speed=AppendVector(A=@n161, B=@n162))) {
   group=SuperLevel_T01
   parameterName=Flowmap
   samplerType=SAMPLERTYPE_LinearColor
   texture=&T_SwirlFlow_01
  }
  NamedRerouteDeclaration(None=Multiply(
    A=ComponentMask(None=@n163.RGB) {g=true, r=true}
    B=ScalarParameter() {defaultValue=0.12, group=SuperLevel_T01, parameterName=DistortIntensity}
   )) {name=NormalDistort_GB, nodeColor=<rgba 0 1 0.6953 1>}
  NamedRerouteDeclaration(None=SmoothStep(Value=@n40.G) {constMax=0.5}) {name=PanningG_GB, nodeColor=<rgba 0.8047 1 0 1>}
  NamedRerouteDeclaration(None=@n45) {name=BG_Projection, nodeColor=<rgba 1 0 0.4922 1>}
  NamedRerouteDeclaration(None=@n42) {name=MainUV_GB, nodeColor=<rgba 1 0 0.375 1>}
  n164 = NamedRerouteUsage()
  n165 = Constant() {r=1}
  n166 = Saturate(None=Multiply(
    A=Desaturation(None=@n164, Fraction=@n165)
    B=VectorParameter() {group=SuperLevel_Control, parameterName=TintEM, defaultValue=<rgba 0 0.4189 0.4306 1>}.RGB
   ))
  n167 = Multiply(A=@n143, B=ScalarParameter() {defaultValue=3, group=SuperLevel, parameterName=OutlinePostMulti})
  n168 = StaticSwitchParameter(True=Multiply(A=Multiply(A=@n141, B=@n167), B=NamedRerouteUsage()), False=Constant()) {defaultValue=true, group=SuperLevel_Control, parameterName=UseOutlineArea}
  n169 = NamedRerouteUsage()
  n170 = Add(A=Add(A=Add(A=@n168, B=@n169), B=NamedRerouteUsage()), B=NamedRerouteUsage())
  n171 = Desaturation(None=NamedRerouteUsage(), Fraction=Constant() {r=1})
  n172 = StaticSwitchParameter(True=LinearInterpolate(A=@n170, B=Multiply(A=@n147, B=@n171), Alpha=NamedRerouteUsage()), False=@n170) {group=SuperLevel_Control, parameterName=UseHairArea}
  n173 = NamedRerouteUsage()
  n174 = ScalarParameter() {defaultValue=0.3, group=SuperLevel_DefaultEM, parameterName=DefaultEM_Min_Distance}
  n175 = NamedRerouteUsage()
  n176 = Multiply(
   A=Add(A=Multiply(A=@n173, B=@n174), B=ScalarParameter() {group=SuperLevel_DefaultEM, parameterName=DefaultEM_Min})
   B=LinearInterpolate(Alpha=@n175)
  )
  n177 = NamedRerouteUsage()
  n178 = NamedRerouteUsage()
  n179 = VectorParameter() {group=SuperLevel_Shirt, parameterName=OutlineColor_RE, defaultValue=<rgba 0.3914 0.401 0 1>}
  n180 = Multiply(A=Desaturation(None=@n178, Fraction=Constant() {r=1}), B=Multiply(A=@n179.RGB, B=@n179.RGB))
  n181 = Add(
   A=Multiply(
    A=VectorParameter() {group=SuperLevel_DefaultEM, parameterName=DefaultEmissive, defaultValue=<rgba 0.8542 0.9772 1 1>}.RGB
    B=Multiply(A=@n176, B=@n177)
   )
   B=Multiply(A=StaticSwitchParameter(True=@n180, False=@n178) {group=SuperLevel_Control, parameterName=UseTintOutlineInput}, B=@n177)
  )
  n182 = ScalarParameter() {defaultValue=1.2, group=SuperLevel_DefaultEM, parameterName=DefaultEM_Multi}
  n183 = StaticSwitchParameter(True=Add(A=@n172, B=Multiply(A=@n181, B=@n182)), False=@n172) {group=SuperLevel_Control, parameterName=UseDefaultEmissiveArea}
  n184 = Desaturation(None=NamedRerouteUsage(), Fraction=Constant() {r=1})
  n185 = ScalarParameter() {defaultValue=1, group=SuperLevel_Control, parameterName=BaseColorMulti}
  n186 = Multiply(A=@n183, B=Saturate(None=Multiply(A=@n184, B=@n185)))
  NamedRerouteDeclaration(None=Add(
    A=StaticSwitchParameter(True=@n166, False=@n164) {group=SuperLevel_Control, parameterName=UseTintORG_EM}
    B=StaticSwitchParameter(True=@n186, False=@n183) {group=SuperLevel_Control, parameterName=BaseColorAffectEM}
   )) {name=EM_OUT, nodeColor=<rgba 0 1 0.8125 1>}
  n187 = VertexInterpolator(VS=PreSkinnedNormal())
  n188 = Saturate(None=DotProduct(A=@n187.PS, B=Constant3Vector() {constant=<rgba 0 1 0 0>}))
  n189 = Saturate(None=DotProduct(A=@n187.PS, B=Constant3Vector() {constant=<rgba 0 -1 0 0>}))
  NamedRerouteDeclaration(None=SmoothStep(
    Min=ScalarParameter() {defaultValue=0.5, group=SuperLevel, parameterName=PreskinnedClampMIN}
    Max=ScalarParameter() {defaultValue=1, group=SuperLevel, parameterName=PreskinnedClampMax}
    Value=Add(A=@n188, B=@n189)
   ) {constMin=0.8}) {name=PreskinnedSeam, nodeColor=<rgba 0 1 0.3438 1>}
  n190 = Saturate(None=SkyLightEnvMapSample(Direction=ReflectionVectorWS(), Roughness=ScalarParameter() {defaultValue=0.1, group=SuperLevel_T01, parameterName=SkyRoughness}))
  NamedRerouteDeclaration(None=@n190) {name=Reflection, nodeColor=<rgba 1 0.6328 0 1>}
  NamedRerouteDeclaration(None=@n47.R) {name=PrimaryArea, nodeColor=<rgba 1 0.5625 0 1>}
  NamedRerouteDeclaration(None=ComponentMask(None=FunctionInput() {bUsePreviewValueAsDefault=true, inputName=M_Map, sortPriority=3, previewValue=<1 1 1 1>}) {b=true}) {name=Skin_Map, nodeColor=<rgba 0 0.2031 1 1>}
  NamedRerouteDeclaration(None=@n47.B) {name=SecondaryArea, nodeColor=<rgba 0 0.6953 1 1>}
  n191 = Fresnel() {baseReflectFraction=0, exponent=1}
  n192 = ComponentMask(None=@n191) {r=true}
  n193 = ScalarParameter() {group=SuperLevel_T01, parameterName=ThinFilmOffset_SL}
  n194 = TextureSampleParameter2D(UVs=AppendVector(A=Add(A=@n192, B=@n193), B=Constant() {r=0.5})) {
   group=SuperLevel_T01
   parameterName=SuperLevelThinFilm
   texture=/BRCosmetics/Characters/Player/Female/Medium/Bodies/F_MED_RoseDepth_Seed/Skins/SuperLevel/Textures/T_SuperLevel_Thinfilm_T01_S33
  }
  n195 = Constant() {r=1}
  n196 = LinearInterpolate(
   A=Desaturation(None=@n194.RGB, Fraction=ScalarParameter() {defaultValue=0.3, group=SuperLevel_T01, parameterName=Destauration})
   B=Multiply(A=@n194.RGB, B=Desaturation(None=@n190, Fraction=@n195))
   Alpha=@n191
  )
  n197 = Saturate(None=Multiply(B=@n196) {constA=2})
  n198 = Multiply(A=@n196, B=ScalarParameter() {defaultValue=0.7, parameterName=BasecolorMinIntensity}) {constB=0.5}
  n199 = NamedRerouteUsage()
  n200 = Sine(None=Multiply(A=Fresnel() {baseReflectFraction=0, exponent=1}))
  n201 = Multiply(
   A=Multiply(
    A=LinearInterpolate(A=@n197, B=@n198, Alpha=@n199)
    B=ScalarParameter() {defaultValue=0.5, group=SuperLevel_T01, parameterName=ThinFilmIntensity_SL}
   )
   B=LinearInterpolate(Alpha=ConstantBiasScale(None=@n200)) {constA=0.2}
  )
  n202 = Multiply(A=NamedRerouteUsage(), B=ScalarParameter() {defaultValue=0.05, parameterName=BasecolorAddIntensity})
  NamedRerouteDeclaration(None=Saturate(None=Add(A=@n201, B=@n202))) {name=ThinFilmColor, nodeColor=<rgba 0.3125 0 1 1>}
  NamedRerouteUsage()
  NamedRerouteUsage()
  Power(Base=@n175)
  Power(Base=NamedRerouteUsage())
  SkyAtmosphereLightDirection()
  StaticSwitchParameter(True=LinearInterpolate(A=@n155, B=NamedRerouteUsage(), Alpha=NamedRerouteUsage()), False=@n155) {group=SuperLevel_Control, parameterName=UseOutlineArea_RE}
 ]}
